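; Shared Tidal Crane swarm NPC pawn logic for storm damage and killer healing, enabled by scalable floats on the server once playlist data is ready.
; A looping timer, cleared on death, tests 2D distance to the safe zone, toggling the in-zone effect and scaling the persistent storm effect's level by phase while outside.
; On death, the killer's player or bot pawn receives a restore GameplayEffect unless downed.
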(event EventBeginPlay
  (Class|FortAsyncActionCurrentPlaylistReady|PlaylistDataReadyAsync
    (:OnReady
      (Networking|SwitchHasAuthority
        (:Authority
          (Utilities|FlowControl|Sequence
            (:then_0
              (if (ScalableFloat|GetValueasBool (Variables|Default|GetShouldTakeStormDamage))
                (CallFunction|HandleBeinginStorm)))
            (:then_1
              (if (ScalableFloat|GetValueasBool (Variables|Default|GetShouldHealOnDeath))
                (EventDispatchers|BindEventtoOnDied (EventDispatchers|CreateEvent :!event Custom|CustomHealOnDeath))))))))))

(event Custom|HandleBeingInStorm
  (bind _aipawnabilitysystemcomponent (Variables|GameplayAbilities|GetAIPawnAbilitySystemComponent))
  ; Safe Zone / Storm Damage Handling
  (bind _returnvalue (Utilities|Time|SetTimerbyEvent (EventDispatchers|CreateEvent :!event Custom|TestIsInStorm) (ScalableFloat|GetValueatLevel (Variables|Default|GetTestStormValueInterval)) true))
  (Variables|InternalData|SetHNDL_SafeZoneValidation _returnvalue)
  (GameplayEffects|ApplyGameplayEffectToSelf :self _aipawnabilitysystemcomponent :GameplayEffectClass (Variables|Default|GetGE_InSafeZone) :Level 0.0)
  (bind _returnvalue_2 (GameplayEffects|ApplyGameplayEffectToSelf :self _aipawnabilitysystemcomponent :GameplayEffectClass (Variables|Default|GetGE_InStorm) :Level 1.0))
  (bind _output_get (Variables|InternalData|SetHNDL_OutsideSafeZoneDmgGE _returnvalue_2))
  (Ability|GameplayEffect|AssignActiveGameplayEffectTagSetbyCallerMagnitude _output_get "(TagName=\"SetByCaller.StormCampingDamage\")" 1.0))

(event Custom|TestIsInStorm
  (bind _aipawnabilitysystemcomponent (Variables|GameplayAbilities|GetAIPawnAbilitySystemComponent))
  (bind _ge_insafezone (Variables|Default|GetGE_InSafeZone))
  ; Safe Zone / Storm Damage Handling
  (bind (_testsucceeded _isinsafezone) (CallFunction|CheckIsinSafeZone))
  (if _testsucceeded
    (if _isinsafezone
      (GameplayEffects|ApplyGameplayEffectToSelf :self _aipawnabilitysystemcomponent :GameplayEffectClass _ge_insafezone :Level 0.0)
      (else
        (GameplayEffects|RemoveActiveGameplayEffectbySourceEffect _aipawnabilitysystemcomponent _ge_insafezone _aipawnabilitysystemcomponent -1)
        (GameplayEffects|SetActiveGameplayEffectLevel _aipawnabilitysystemcomponent (Variables|InternalData|GetHNDL_OutsideSafeZoneDmgGE) (select (Utilities|IsValid (SafeZone|GetSafeZoneIndicator) :!class /Script/Engine.KismetSystemLibrary) (SafeZone|GetCurrentPhaseIndex (SafeZone|GetSafeZoneIndicator)) 0))))))

(event EventOnDeathServer
  (params (Damage :type Float%20%28single-precision%29) (DamageTags :type Gameplay%20Tag%20Container%20Structure%20%28by%20ref%29) (Momentum :type Vector :default "0, 0, 0") (HitInfo :type Hit%20Result%20Structure%20%28by%20ref%29) (InstigatedBy :type Controller%20Object%20Reference) (DamageCauser :type Actor%20Object%20Reference) (EffectContext :type Gameplay%20Effect%20Context%20Handle%20Structure))
  (|CallParentFunction Damage DamageTags Momentum HitInfo InstigatedBy DamageCauser EffectContext)
  (Utilities|Time|ClearandInvalidateTimerbyHandle (Variables|InternalData|GetHNDL_SafeZoneValidation)))

(event Custom|CustomHealOnDeath
  (params (DamagedActor :type Actor%20Object%20Reference) (Damage :type Float%20%28single-precision%29) (InstigatedBy :type Controller%20Object%20Reference) (DamageCauser :type Actor%20Object%20Reference) (HitLocation :type Vector) (FHitComponent :type Primitive%20Component%20Object%20Reference) (BoneName :type Name) (Momentum :type Vector))
  ; Allow to heal on Death
  (if (Fortnite|IsSpatialGameplayExperience)
    (if (GameplayTags|HasMatchingGameplayTag self "(TagName=\"Pawn.Athena.NPC.Deimos.DisableShieldOnElimination\")")
      _
      (else
        _))
    (else
      _))
  (bind _asfort_player_controller_athena (Utilities|Casting|CastToFortPlayerControllerAthena InstigatedBy)
    (:then
      (bind _output_get (Variables|Default|SetCustomHeal_KillerPawn (Game|Player|GetPlayerPawn _asfort_player_controller_athena)))
      (Variables|Default|SetHealOnDeath_KillerAbilitySystem (Ability|GetAbilitySystemComponent _output_get)))
    (:CastFailed
      (bind _asfort_athena_aibot_controller (Utilities|Casting|CastToFortAthenaAIBotController InstigatedBy))
      (bind _output_get_2 (Variables|Default|SetCustomHeal_KillerPawn (Variables|/Script/FortniteAI.FortAthenaAIBotController|GetPlayerBotPawn _asfort_athena_aibot_controller)))
      (Variables|Default|SetHealOnDeath_KillerAbilitySystem (Ability|GetAbilitySystemComponent _output_get_2))))
  (Utilities|IsValid (Variables|Default|GetHealOnDeath_KillerAbilitySystem)
    (:Is%20Valid
      (if (not (Health|IsDBNO (Variables|Default|GetCustomHeal_KillerPawn)))
        (GameplayEffects|ApplyGameplayEffectToSelf :self (Variables|Default|GetHealOnDeath_KillerAbilitySystem) :GameplayEffectClass (Variables|Default|GetGE_RestoreOnBugKill) :Level 0.0)))))

(fn CheckIsInSafeZone
  (bind _returnvalue (SafeZone|GetSafeZoneIndicator))
  (if (Utilities|ClassIsChildOf (Utilities|GetClass (Game|GetGameState)) (Utilities|ResolveSoftReference (Variables|Default|GetAthena_GameState_SoftClass)))
    (Utilities|IsValid _returnvalue
      (:Is%20Valid
        (return true (< (Math|Vector|Distance2DSquared(Vector) (SafeZone|GetSafeZoneCenter _returnvalue) (Transformation|GetActorLocation)) (Math|Float|Square (SafeZone|GetSafeZoneRadius _returnvalue)))))
      (:Is%20Not%20Valid
        (return false false)))
    (else
      (return false false))))

(fn ConstructionScript)
